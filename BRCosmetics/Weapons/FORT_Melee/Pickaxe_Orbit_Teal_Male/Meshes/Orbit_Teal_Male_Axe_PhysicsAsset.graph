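# Physics asset for the Orbit Teal Male two-handed pickaxe mesh, a Battle Royale melee cosmetic.
# It holds a single body on the Melee_TwoHanded_Handle bone, shaped as one offset, rotated capsule that serves as a simple collision approximation of the weapon.
# It defines no constraints, so there is no jointed or secondary physics chain.
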PhysicsAsset /BRCosmetics/Weapons/FORT_Melee/Pickaxe_Orbit_Teal_Male/Meshes/Orbit_Teal_Male_Axe_PhysicsAsset
bodies {Melee_TwoHanded_Handle [{shapeType=Capsule, length=69.5831, radius=14.5281, center=<0.8076 5.8556 20.4705>, rotation=<rot 5.5878 -90 0.0017>}]}
constraints []
